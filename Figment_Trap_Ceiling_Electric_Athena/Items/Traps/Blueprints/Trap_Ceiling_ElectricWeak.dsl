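; Weak ceiling electric trap that handles its trigger setup and the audio and visual feedback across its lifecycle.
; Placement adds the trigger, dust and sound, while build completion, world ready and reload end start idle sparks and charge hum.
; The activation GameplayCue plays fire effects, stops the hum, intensifies sparks and drives light intensity by timeline, then stops them.
; Construction hides the light and gates its shadows by effects quality; exhausted durability plays a delayed explosion sound.
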(event EventOnPlaced
  ;[1 1 1] Initial Setup
  (Trap|AddTriggerComponent (Variables|Trap_Ceiling_Electric|GetTriggerComponent))
  (Audio|PlaySoundatLocation (Variables|Audio|GetTrap_Placed_Sound) (Transformation|GetActorLocation))
  (Utilities|TrapPlacementDustFX :Attach%20to%20Component (Variables|Debug|GetStaticMeshComponent) :Socket "PlacedFX" :Rotation "0, 0, 0" :Burst%20Rate 5 :Size%20%28Volume%29 "30.000000,110.000000,15.000000" :Velocity%20Scale "1,1,1" :Size%20Scale%20%28Particles%29 1))

(event EventOnFinishedBuilding
  ;[1 1 1] Finished building
  (Components|Activation|Activate (Variables|Trap_Ceiling_Electric|GetCharge%20Sound) false)
  (Components|Activation|Activate (Variables|Trap_Ceiling_Electric|GetIdleElectricFX) false))

(event EventOnWorldReady
  ;[1 1 1] Finished building
  (Components|Activation|Activate (Variables|Trap_Ceiling_Electric|GetCharge%20Sound) false)
  (Components|Activation|Activate (Variables|Trap_Ceiling_Electric|GetIdleElectricFX) false))

(event GameplayCue|GameplayCue.Abilities.Activation.Traps.ActivateTrap
  (params (EventType :type EGameplayCueEvent%20Enum :default "OnActive") (Parameters :type Gameplay%20Cue%20Parameters%20Structure%20%28by%20ref%29))
  ;[1 1 1] On Fire
  (switch Utilities|FlowControl|Switch|SwitchonEGameplayCueEvent EventType
    (:Executed
      (Audio|PlaySoundatLocation (Variables|Audio|GetTrap_Fire_Sound) (Transformation|GetActorLocation))
      (Effects|SpawnEmitterAttached "/Figment_Trap_Ceiling_Electric_Athena/Weapons/FORT_TRAPS/Electric_Trap/Particles/P_ElectricWeakCeiling_Trap_Fire_01.P_ElectricWeakCeiling_Trap_Fire_01" (Variables|Debug|GetStaticMeshComponent) "FireFX")
      (Components|Activation|Deactivate (Variables|Trap_Ceiling_Electric|GetCharge%20Sound))
      (Development|SetHiddeninGame (Variables|Trap_Ceiling_Electric|GetLight) false false)
      (Effects|Components|ParticleSystem|SetFloatParameter (Variables|Trap_Ceiling_Electric|GetIdleElectricFX) "SpawnRateScale" 5)
      (Variables|/Script/CascadeRuntime.ParticleSystemComponent|SetCustomTimeDilation 3 (Variables|Trap_Ceiling_Electric|GetIdleElectricFX))
      (bind (_direction _brightness) (Timeline|ScaleOutLight PlayFromStart)
        (:Update
          (Rendering|Components|Light|SetIntensity (Variables|Trap_Ceiling_Electric|GetLight) _brightness))
        (:Finished
          (Development|SetHiddeninGame (Variables|Trap_Ceiling_Electric|GetLight) false false)
          (Effects|Components|ParticleSystem|SetFloatParameter (Variables|Trap_Ceiling_Electric|GetIdleElectricFX) "SpawnRateScale" 1)
          (Variables|/Script/CascadeRuntime.ParticleSystemComponent|SetCustomTimeDilation 1 (Variables|Trap_Ceiling_Electric|GetIdleElectricFX))
          (Components|Activation|Deactivate (Variables|Trap_Ceiling_Electric|GetIdleElectricFX)))))))

(event EventOnReloadEnd
  ;[1 1 1] Reload End
  (Components|Activation|Activate (Variables|Trap_Ceiling_Electric|GetIdleElectricFX) false)
  (Components|Activation|Activate (Variables|Trap_Ceiling_Electric|GetCharge%20Sound) false))

(event EventOnOutOfDurability
  ;[1 1 1] Durability Change
  (Utilities|FlowControl|Delay 1)
  (Audio|PlaySoundatLocation (Variables|Audio|GetExplosion_Sound) (Transformation|GetActorLocation)))

(fn ConstructionScript
  (Networking|CanExecuteCosmeticEvents
    (:True
      (Development|SetHiddeninGame (Variables|Trap_Ceiling_Electric|GetLight) true false)
      (bind (_vfx_quality_int _quality_check_bool) (Utilities|EffectsQualityCheck 2)
        (:Cinematic
          _)
        (:Epic
          _)
        (:High
          _)
        (:Medium
          _)
        (:Low
          _))
      (Rendering|Components|Light|SetCastShadows (Variables|Trap_Ceiling_Electric|GetLight) _quality_check_bool))))
